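# Sound cue for the Jupiter emote's engine audio, played as emote music with emote attenuation and concurrency, plus a send to the music analysis submix.
# A mixer layers one of three random start one-shots, at reduced volume, with a looping engine branch.
# The loop branch starts after a short delay, crossfades idle and acceleration loops via the EmoteMoveSpeed parameter, and passes through a rising volume envelope and an emote music sound class override.
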SoundCue /BRCosmetics/Sounds/Emotes/Jupiter/Emote_Jupiter_Engine_Cue
root=Mixer_0
nodes [
 {id=Delay_0, children [ParamCrossFade_0], properties {delayMax=0.7, delayMin=0.7}}
 {
  id=Enveloper_0
  children [Delay_0]
  properties {pitchCurve {editorCurveData {keys [{time=0}]}}, volumeCurve {editorCurveData {keys [<at -0.0046 0.0182>,<at 0.42 0.0295>,<at 1 1.0584>]}}}
 }
 {id=Mixer_0, children [SoundClass_0,Random_0], properties {inputVolume [1,0.45]}}
 {
  id=ParamCrossFade_0
  children [WavePlayer_2,WavePlayer_7]
  properties {
   paramName=EmoteMoveSpeed
   crossFadeInput [
    {fadeInDistanceEnd=0, fadeInDistanceStart=0, fadeOutDistanceEnd=1, fadeOutDistanceStart=0, volume=0.5}
    {fadeInDistanceEnd=1, fadeInDistanceStart=0, fadeOutDistanceEnd=100, fadeOutDistanceStart=100, volume=1}
   ]
  }
 }
 {id=Random_0, children [WavePlayer_0,WavePlayer_1,WavePlayer_3], properties {weights [1,1,1]}}
 {id=SoundClass_0, children [Enveloper_0], properties {soundClassOverride=&Fort_Music_Emote}}
 {id=WavePlayer_0, properties {soundWaveAssetPtr=&/Emote_Jupiter_Start_01}}
 {id=WavePlayer_1, properties {soundWaveAssetPtr=&/Emote_Jupiter_Start_02}}
 {id=WavePlayer_2, properties {bLooping=true, soundWaveAssetPtr=&/Emote_Jupiter_IdleLoop_01}}
 {id=WavePlayer_3, properties {soundWaveAssetPtr=&/Emote_Jupiter_Start_03}}
 {id=WavePlayer_7, properties {bLooping=true, soundWaveAssetPtr=&/Emote_Jupiter_AccelLoop_03}}
]
properties {
 attenuationSettings=&Fort_Emotes_Athena_Att
 maxDistance=1600
 soundClassObject=&Fort_Music_Emote
 volumeMultiplier=1
 concurrencyOverrides {resolutionRule=PreventNew}
 concurrencySet [/Game/Sounds/Concurrency_Rules/Athena_EmoteMusic_Conc]
 soundSubmixSends [{minSendLevel=0, soundSubmix=&Fortnite_AudioAnalysis_Music}]
}
